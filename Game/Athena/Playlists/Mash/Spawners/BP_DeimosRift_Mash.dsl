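; Mash mode Deimos rift that grants a score-multiplier ability set and sends its Opened event when killed by a player, possibly spawning a chest.
; On authority it spawns a damage-proxy weakpoint actor, because the parent rift lacks pickaxe weakpoints, whose health drives damage visuals.
; A lifetime timer, refreshed by throttled damage, expires the rift.
; Optional overlap handling pops old chests and tosses pickups aside, briefly ignoring their collision.
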(event EventBeginPlay
  (bind _self self)
  ;[0 1 0.1512] EVENT -> BeginPlay
  (|CallParentFunction)
  (Networking|SwitchHasAuthority
    (:Authority
      (Utilities|FlowControl|Sequence
        (:then_0
          (CallFunction|SetLifeTimeTimer)
          (Ability|EquipFortAbilitySet self "/Game/Athena/Playlists/Mash/GameAbilities/GAS_ScoreMultiplier.GAS_ScoreMultiplier" (Variables|Default|GetAbilitySystem))
          (GameplayEffects|ApplyGameplayEffectToSelf :self (Ability|GetAbilitySystemComponent self) :GameplayEffectClass "/Game/Athena/Playlists/Mash/GameEffects/GE_Mash_RiftHealthMultiplier.GE_Mash_RiftHealthMultiplier_C" :Level 0.0)
          ;;[0 1 0.6735] Spawn a ABuildingProp_DamageProxy (BP_DemiosRiftMash_WeakPointers)
          ;; The spawned actor will take damage events and repeat them to this actor.
          ;; We do this because ABuildingProp_DamageProxy supports pickaxe weakpoints and our parent class ABuildingRift does NOT.
          (Utilities|FlowControl|DelayUntilNextTick) ; We have to wait 1 frame after BeginPlay to spawn buildings near the rift,\notherwise ABuildingRift::DestroyNearbyBuildings might get called and destroy the newly spawned building.
          (Game|SpawnActorfromClass (Variables|Default|GetWeakpointClass) (Transformation|GetWorldTransform (Variables|Default|GetCubeMesh)) "AlwaysSpawn" "MultiplyWithRoot" _self 0)
          (Building|InitializeKismetSpawnedBuildingActor _self 0 false) ; This needs to be called on BP spawned buildings or else the spawned building  can experience bugs like not always replicating.
          (Variables|Default|SetWeakPoint%20Actor 0))
        (:then_1
          ;;[0.2307 0.6172 0.9301] Check for and handle any overlapping chests or pickups.
          ;; This can occur if we need to reuse a spawner location
          (CallFunction|HandleOverlappingLoot))))))

(event Custom|SetLifeTimeTimer
  ;[1 1 1] CUSTOM EVENT -> SetLifeTimeTimer
  (Utilities|Time|ClearandInvalidateTimerbyHandle (Variables|Default|GetLifeTimeTimerHandle))
  (bind _returnvalue (Utilities|Time|SetTimerbyEvent (EventDispatchers|CreateEvent :!event Custom|LifeTimeExpired) (Variables|Default|GetLifeTime)))
  (Variables|Default|SetLifeTimeTimerHandle _returnvalue))

(event Custom|HandleOverlappingLoot
  ;[0.2307 0.6172 0.9301] HandleOverlappingLoot
  (if (ScalableFloat|GetValueasBool (Variables|Default|GetHF_TossOverlappingLoot))
    (Networking|SwitchHasAuthority
      (:Authority
        (Utilities|FlowControl|Sequence
          (:then_0
            ;;[0.5663 0.3828 1 0.502] Check for overlapping chests. If this spawner is reusing a spawn location, there may be a chest from a previous spawner there.
            ;; If so, bust it open.
            (bind _outactors (Collision|SphereOverlapActors (Transformation|GetActorLocation) (Math|Float|Max(Float) (.x (Components|Box|GetScaledBoxExtent (Variables|Default|GetNoBuildZone))) (.y (Components|Box|GetScaledBoxExtent (Variables|Default|GetNoBuildZone)))) (Utilities|Array|MakeArray "ObjectTypeQuery1") "/Game/Athena/Playlists/Mash/Chest_Athena_Mash.Chest_Athena_Mash_C" (Utilities|Array|MakeArray self)))
            (for _array_element _outactors
              (bind _aschest_athena_mash (Utilities|Casting|CastToChest_Athena_Mash _array_element))
              (Container|SpawnLoot _aschest_athena_mash 0)
              (Gameplay|SilentDie _aschest_athena_mash false)))
          (:then_1
            ;;[0.1918 0.1018 0.4045 0.502] Short wait, then check for pickups and toss them out of the way. Do this even if we didn't open a chest as there may be pickups under us either way.
            (bind _returnvalue (Utilities|Time|SetTimerbyEvent (delegate Custom|PickupsOverlap) (Variables|Default|GetDelayBeforePickupOverlap)))
            (Variables|Default|SetPickupsOverlapTimerHandle _returnvalue)))))))

(event EventOnDamageServer
  (params (Damage :type Float%20%28single-precision%29) (DamageTags :type Gameplay%20Tag%20Container%20Structure%20%28by%20ref%29) (Momentum :type Vector :default "0, 0, 0") (HitInfo :type Hit%20Result%20Structure%20%28by%20ref%29) (InstigatedBy :type Controller%20Object%20Reference) (DamageCauser :type Actor%20Object%20Reference) (EffectContext :type Gameplay%20Effect%20Context%20Handle%20Structure))
  ;[1 1 1] EVENT -> OnDamageServer
  (if (> (Utilities|TimeSince (Variables|Default|GetTimeLastSetLifeTimeTimer)) (Variables|Default|GetThrottleTime))
    (Variables|Default|SetTimeLastSetLifeTimeTimer (Utilities|Time|GetTimeSeconds))
    (CallFunction|SetLifeTimeTimer)))

(event Custom|PickupsOverlap
  ;[0.2307 0.6172 0.9301] HandleOverlappingLoot
  ;;[0.1918 0.1018 0.4045 0.502] Short wait, then check for pickups and toss them out of the way. Do this even if we didn't open a chest as there may be pickups under us either way.
  (Utilities|FlowControl|Sequence
    (:then_0
      (bind _outactors (Collision|SphereOverlapActors (Transformation|GetActorLocation) (Math|Float|Max(Float) (.x (Components|Box|GetScaledBoxExtent (Variables|Default|GetNoBuildZone))) (.y (Components|Box|GetScaledBoxExtent (Variables|Default|GetNoBuildZone)))) (Utilities|Array|MakeArray "ObjectTypeQuery4") "/Script/FortniteGame.FortPickupAthena" (Utilities|Array|MakeArray self)))
      (for _array_element _outactors
        (bind _asfort_pickup_athena (Utilities|Casting|CastToFortPickupAthena _array_element))
        (Collision|IgnoreActorwhenMoving (Variables|/Script/FortniteGame.FortPickup|GetTouchCapsule _asfort_pickup_athena) self true)
        (Collision|IgnoreActorwhenMoving (Variables|/Script/FortniteGame.FortPickup|GetTouchCapsule _asfort_pickup_athena) (Variables|Default|GetWeakPoint%20Actor) true)
        (Utilities|Array|Add (Variables|Default|GetIgnoredPickups) _asfort_pickup_athena)
        (Pickup|TossPickupinDirection _asfort_pickup_athena (* (Math|Vector|Normalize (- (+ (Transformation|GetActorLocation _asfort_pickup_athena) (Variables|Default|GetPickupTossOffset)) (Transformation|GetActorLocation))) (Variables|Default|GetPickupTossSpeed)) 0 false false) ; Toss the pickup away from the center of the rift
      ))
    (:then_1
      ;;[0.0848 0.04501 0.1788 0.502] Short wait, then reset the "IgnoreActorWhenMoving" on any pickups we tossed
      (if (Utilities|Array|IsNotEmpty (Variables|Default|GetIgnoredPickups))
        (bind _returnvalue (Utilities|Time|SetTimerbyEvent (delegate Custom|ClearIgnoredPickups) (Variables|Default|GetIgnoreTossedPickupsTime)))
        (Variables|Default|SetClearIgnoredPickupsTimerHandle _returnvalue)))))

(event Custom|LifeTimeExpired
  ;[1 1 1] CUSTOM EVENT -> LifeTimeExpired
  (GameplayEffects|ApplyGameplayEffectToSelf :self (Ability|GetAbilitySystemComponent self) :GameplayEffectClass (Variables|GameplayEffects|GetGE_LifeTimeLapsed) :Level 0.0)
  (Actor|DestroyActor))

(event EventBP_Cosmetic_Intro
  ;[0.5937 1 0.5123] Replicated Cosmetic Events
  ; Calling SetCosmeticState replicates to these.
  (|CallParentFunction)
  (Development|PrintString "BP_DeimosRift_Dynamic: Cosmetic Intro" false false))

(event EventBP_Cosmetic_Idle
  ;[0.5937 1 0.5123] Replicated Cosmetic Events
  ; Calling SetCosmeticState replicates to these.
  (|CallParentFunction)
  (Development|PrintString "BP_DeimosRift_Dynamic: Cosmetic Idle" false false))

(event Custom|ClearIgnoredPickups
  ;[0.2307 0.6172 0.9301] HandleOverlappingLoot
  ;;[0.0848 0.04501 0.1788 0.502] Short wait, then reset the "IgnoreActorWhenMoving" on any pickups we tossed
  (for _array_element (Variables|Default|GetIgnoredPickups)
    (Utilities|IsValid _array_element
      (:Is%20Valid
        (Collision|IgnoreActorwhenMoving (Variables|/Script/FortniteGame.FortPickup|GetTouchCapsule _array_element) self false)
        (Collision|IgnoreActorwhenMoving (Variables|/Script/FortniteGame.FortPickup|GetTouchCapsule _array_element) (Variables|Default|GetWeakPoint%20Actor) false))))
  (Utilities|Array|Clear (Variables|Default|GetIgnoredPickups)))

(event EventBP_Cosmetic_RampUp
  ;[0.5937 1 0.5123] Replicated Cosmetic Events
  ; Calling SetCosmeticState replicates to these.
  (|CallParentFunction)
  (Development|PrintString "BP_DeimosRift_Dynamic: Cosmetic Ramp Up" false false))

(event EventBP_Cosmetic_ShouldDie
  ;[0.5937 1 0.5123] Replicated Cosmetic Events
  ; Calling SetCosmeticState replicates to these.
  (|CallParentFunction)
  (Development|PrintString "BP_DeimosRift_Dynamic: Cosmetic Should Die" false false))

(event Custom|LootOverlapCleanup
  ;[0.2307 0.6172 0.9301] HandleOverlappingLoot
  ;; Overlap Cleanup (for death and endplay)
  (Utilities|Time|ClearandInvalidateTimerbyHandle (Variables|Default|GetPickupsOverlapTimerHandle))
  (Utilities|Time|ClearandInvalidateTimerbyHandle (Variables|Default|GetClearIgnoredPickupsTimerHandle))
  (CallFunction|ClearIgnoredPickups))

(event EventOnDeathServer
  (params (Damage :type Float%20%28single-precision%29) (DamageTags :type Gameplay%20Tag%20Container%20Structure%20%28by%20ref%29) (Momentum :type Vector :default "0, 0, 0") (HitInfo :type Hit%20Result%20Structure%20%28by%20ref%29) (InstigatedBy :type Controller%20Object%20Reference) (DamageCauser :type Actor%20Object%20Reference) (EffectContext :type Gameplay%20Effect%20Context%20Handle%20Structure))
  ;[1 0.5049 0] On Death : Chance to spawn a Chest with extra stuff in it!
  (|CallParentFunction Damage DamageTags Momentum HitInfo InstigatedBy DamageCauser EffectContext)
  (CallFunction|LootOverlapCleanup)
  (Utilities|IsValid InstigatedBy
    (:Is%20Valid
      (Ability|SendGameplayEventtoActor self "(TagName=\"Athena.LTM.Mash.ScoreMultiplier.Opened\")" (Utilities|Struct|MakeGameplayEventData :Instigator self :OptionalObject InstigatedBy :ContextHandle "()" :InstigatorTags "(GameplayTags=)" :TargetTags "(GameplayTags=)" :EventMagnitude 0.0 :TargetData "()"))
      (bind _outxy (Fortnite|EvaluateCurveTableRow (Variables|Default|GetChanceForChestToSpawn)))
      (if (<= (Math|Random|RandomFloat) _outxy)
        (bind _returnvalue (Game|SpawnActorfromClass "/Game/Athena/Playlists/Mash/Chest_Athena_Mash.Chest_Athena_Mash_C" (Math|Transform|MakeTransform (Math|Vector|MakeVector (.x (.location (Transformation|GetWorldTransform (Variables|Default|GetCubeMesh)))) (.y (.location (Transformation|GetWorldTransform (Variables|Default|GetCubeMesh)))) (+ (.z (.location (Transformation|GetWorldTransform (Variables|Default|GetCubeMesh)))) (Variables|Default|GetChestVerticalOffset%20Const))) (.rotation (Transformation|GetWorldTransform (Variables|Default|GetCubeMesh))) (.scale (Transformation|GetWorldTransform (Variables|Default|GetCubeMesh)))) "AdjustIfPossibleButAlwaysSpawn" "MultiplyWithRoot" 0 0 0 0))
        (Building|InitializeKismetSpawnedBuildingActor _returnvalue 0 0 false 0 false) ; This needs to be called on BP spawned buildings or else the spawned building  can experience bugs like not always replicating.
      ))))

(event EventEndPlay
  (params (EndPlayReason :type EEndPlayReason%20Enum :default "Destroyed"))
  ;[0.4948 0.2168 0.1166] EndPlay
  (|CallParentFunction EndPlayReason)
  (CallFunction|LootOverlapCleanup))

(event EventOnDamaged_Bind
  (params (DamagedActor :type Actor%20Object%20Reference) (Damage :type Float%20%28single-precision%29) (InstigatedBy :type Controller%20Object%20Reference) (DamageCauser :type Actor%20Object%20Reference) (HitLocation :type Vector :default "0, 0, 0") (FHitComponent :type Primitive%20Component%20Object%20Reference) (BoneName :type Name :default "None") (Momentum :type Vector :default "0, 0, 0"))
  (bind _weakpoint_actor (Variables|Default|GetWeakPoint%20Actor))
  ;[1 0.1375 0.6715] Update the damage state of the rift based on the WeakPointActor's health (since it acts as a damage proxy for this actor).
  (Utilities|IsValid _weakpoint_actor
    (:Is%20Valid
      (CallFunction|UpdateDamageState (Health|GetHealthPercent _weakpoint_actor)))
    (:Is%20Not%20Valid
      (|CallParentFunction DamagedActor Damage InstigatedBy DamageCauser HitLocation FHitComponent BoneName Momentum))))

(fn GetSpawnMeshes
  (bind _weakpoint_actor (Variables|Default|GetWeakPoint%20Actor))
  ;[1 0.1768 0.8859] Add WeakPointActor's static mesh component to the list of spawn meshes we get from Parent::GetSpawnMeshes
  (bind _meshes (|CallParentFunction))
  (bind _output_get (LocalVariables|SetOut%20Spawn%20Static%20Mesh%20Components _meshes))
  (Utilities|IsValid _weakpoint_actor
    (:Is%20Valid
      (Utilities|Array|Add _output_get (Class|BPDeimosRiftMashWeakPoints|GetStaticMeshComponent _weakpoint_actor)))
    (:Is%20Not%20Valid
      _))
  (return (LocalVariables|GetOut%20Spawn%20Static%20Mesh%20Components)))

(fn OnRep_ServerSpawnTime)

(fn ConstructionScript
  (bind _p_deimos_riftspawn_damagestate (Variables|Default|GetP_Deimos_RiftSpawn_DamageState))
  (|CallParentFunction)
  (Effects|Components|ParticleSystem|SetVectorParameter _p_deimos_riftspawn_damagestate "SizeScale" "0.700000,0.700000,0.700000")
  (Effects|Components|ParticleSystem|SetFloatParameter _p_deimos_riftspawn_damagestate "Radius" 0.0)
  (Components|DestroyComponent (Variables|Default|GetSM_CubeRunes)))
