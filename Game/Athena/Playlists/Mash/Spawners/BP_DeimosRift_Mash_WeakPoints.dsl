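; Weak point actor for the Deimos rift in the Mash playlist, which filters out damage and death checks from NPC attackers.
; On BeginPlay, where cosmetics can run, it copies the owning rift's cube materials onto its static mesh.
; BlueprintModifyIncomingDamage and ShouldDie use controller casts to tell NPCs from game-participant bots.
; The ShouldDie override works around AI destructibility tests that would otherwise make the rift kill itself.
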(event EventBeginPlay
  (bind _staticmeshcomponent (Variables|Debug|GetStaticMeshComponent))
  ;[0.04113 1 0] BeginPlay: Update cosmetics based on our owners settings.
  (|CallParentFunction)
  (Networking|CanExecuteCosmeticEvents
    (:True
      (bind _asbp_deimos_rift_mash (Utilities|Casting|CastToBP_DeimosRift_Mash (Actor|GetOwner)))
      (Rendering|Material|SetMaterial _staticmeshcomponent 0 (Variables|/Game/Athena/Deimos/Spawners/RiftSpawners/BP_DeimosRift.BP_DeimosRift_C|GetCubeMat _asbp_deimos_rift_mash))
      (Rendering|Material|SetMaterial _staticmeshcomponent 1 (Variables|/Game/Athena/Deimos/Spawners/RiftSpawners/BP_DeimosRift.BP_DeimosRift_C|GetCubeMatLOD1 _asbp_deimos_rift_mash)))))

(fn BlueprintModifyIncomingDamage
  (params (Damage :type Float%20%28single-precision%29) (InTags :type Gameplay%20Tag%20Container%20Structure%20%28by%20ref%29) (EffectContext :type Gameplay%20Effect%20Context%20Handle%20Structure%20%28by%20ref%29) (EventInstigator :type Controller%20Object%20Reference) (DamageCauser :type Actor%20Object%20Reference))
  ;[0.5245 0.9234 1] If the damage instigator is an NPC (AthenaAIController or non game participant FortAthenaAIBotController), ignore the damage
  (bind _returnvalue (|CallParentFunction Damage InTags EffectContext EventInstigator DamageCauser))
  (Utilities|Casting|CastToAthenaAIController EventInstigator
    (:then
      _)
    (:CastFailed
      (bind _asfort_athena_aibot_controller (Utilities|Casting|CastToFortAthenaAIBotController EventInstigator)
        (:then
          (if (not (Class|FortPlayerStateAthena|ControllerIsAnAthenaGameParticipant _asfort_athena_aibot_controller))
            (return 0.0)
            (else
              _)))
        (:CastFailed
          _))
      (return _returnvalue))))

(fn GetStaticMeshComponent
  (return (Variables|Debug|GetStaticMeshComponent)))

(fn ShouldDie
  (params (Damage :type Float%20%28single-precision%29) (EventInstigator :type Controller%20Object%20Reference) (DamageCauser :type Actor%20Object%20Reference) (bIsTest :type Boolean) (EffectContext :type Gameplay%20Effect%20Context%20Handle%20Structure%20%28by%20ref%29))
  ;[0.5245 0.9234 1] Building Rifts do something weird and kill themselves when ShouldDie is tested on them (think they're blocked?).  The new "bot" AI use ShouldDie to test if they can destroy a building  to determine if they should target it in order to get to a player hiding behind it, so if this is an NPC instigating this ShouldDie damage, skip the parent ShouldDie which would kill the rift.
  (Utilities|Casting|CastToAthenaAIController EventInstigator
    (:then
      _)
    (:CastFailed
      (bind _asfort_athena_aibot_controller (Utilities|Casting|CastToFortAthenaAIBotController EventInstigator)
        (:then
          (if (not (Class|FortPlayerStateAthena|ControllerIsAnAthenaGameParticipant _asfort_athena_aibot_controller))
            (Development|PrintString "Blocking ShouldDie")
            (return false)
            (else
              _)))
        (:CastFailed
          _))
      (bind _returnvalue (|CallParentFunction Damage EventInstigator DamageCauser false EffectContext))
      (return _returnvalue))))

(fn ConstructionScript)
