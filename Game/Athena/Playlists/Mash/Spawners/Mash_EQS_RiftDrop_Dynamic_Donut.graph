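# EQS query for Mash playlist rift drops that generates candidate points on a donut arc around the current safe zone indicator center.
# A single non-spiral ring spans a 45-degree arc aimed from that center toward the predicted next safe zone location.
# Points project vertically onto the navmesh with the default-area filter, and no tests score or filter them.
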EnvQuery /Game/Athena/Playlists/Mash/Spawners/Mash_EQS_RiftDrop_Dynamic_Donut
options [{
  generator {
   name=EnvQueryGenerator_Donut_0
   class=EnvQueryGenerator_Donut
   properties {
    arcAngle=45
    bDefineArc=true
    bUseSpiralPattern=false
    center=/Script/FortniteAIServer.FortQueryContext_AthenaCurrentSafeZoneIndicatorCenter
    innerRadius=6000
    numberOfRings=1
    outerRadius=10000
    pointsPerRing=20
    arcDirection {
     lineFrom=&FortQueryContext_AthenaCurrentSafeZoneIndicatorCenter
     lineTo=/Script/FortniteAIServer.FortQueryContext_AthenaSafeZonePredictedLocation
    }
    projectionData {
     extentX=50
     navigationFilter=&RecastFilter_UseDefaultArea
     postProjectionVerticalOffset=0
     projectDown=3000
     projectUp=10000
    }
   }
  }
  tests []
 }]
