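# EQS query for the Mash playlist spawners that picks rift drop locations inside the current safe zone.
# A donut generator lays spiral rings of points around the safe zone center, with its arc direction taken from a random XY direction context.
# Tests evaluate flat ground without navmesh, surface validity and nearby buildings.
# Inside-building and box-trace obstruction tests act as filters on the candidates.
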EnvQuery /Game/Athena/Playlists/Mash/Spawners/Mash_EQS_RiftDrop_Static_Donut_InZone
options [{
  generator {
   name=EnvQueryGenerator_Donut_1
   class=EnvQueryGenerator_Donut
   properties {
    bUseSpiralPattern=true
    center=/Script/FortniteAIServer.FortQueryContext_AthenaCurrentSafeZoneCenter
    innerRadius=1500
    numberOfRings=4
    outerRadius=7000
    pointsPerRing=6
    arcDirection {
     dirMode=Rotation
     lineFrom=&FortQueryContext_AthenaCurrentSafeZoneCenter
     rotation=/Script/FortniteAIServer.FortQueryContext_RandomDirectionXY
    }
    projectionData {
     extentX=50
     navigationFilter=&RecastFilter_UseDefaultArea
     postProjectionVerticalOffset=0
     projectDown=3000
     projectUp=10000
     serializedChannel=ECC_GameTraceChannel11
     traceChannel=TraceTypeQuery10
    }
   }
  }
  tests [
   FortQueryTest_OnFlatSurfaceNoNavMesh_0 {
    class=FortQueryTest_OnFlatSurfaceNoNavMesh
    properties {
     bDefineReferenceValue=false
     boolValue=false
     clampMaxType=None
     clampMinType=None
     filterType=Match
     floatValueMax=0
     floatValueMin=0
     multipleContextFilterOp=AllPass
     multipleContextScoreOp=AverageScore
     normalTolerance=15
     normalizationType=Absolute
     radius=200
     referenceValue=0
     scoreClampMax=0
     scoreClampMin=0
     scoringEquation=Constant
     scoringFactor=100
     testComment=""
     testPurpose=Score
     traceOffset=500
     zTolerance=30
     actorClassesToIgnoreInTrace []
    }
   }
   FortQueryTest_HasNearbyBuildings_0 {class=FortQueryTest_HasNearbyBuildings, properties {extentXY=1, extentZ=2, scoringFactor=20, testPurpose=FilterAndScore}}
   FortQueryTest_ValidSurface_0 {
    class=FortQueryTest_ValidSurface
    properties {flatSurfaceToleranceZ=50, radius=360, traceOffsetDown=100, traceOffsetUp=100, surfaceMaterials []}
   }
   FortQueryTest_InsideBuilding_1 {class=FortQueryTest_InsideBuilding, properties {testPurpose=Filter}}
   FortQueryTest_IsObstructed_1 {
    class=FortQueryTest_IsObstructed
    properties {
     boolValue=true
     context=&EnvQueryContext_Querier
     itemHeightOffset=150
     testPurpose=Filter
     traceData {extentX=250, extentY=250, extentZ=300, traceShape=Box}
    }
   }
  ]
 }]
