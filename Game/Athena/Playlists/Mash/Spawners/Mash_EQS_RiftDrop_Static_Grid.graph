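# Environment query for the Mash playlist that picks rift drop spawner locations, judging by its name, from a static grid around the current safe zone center.
# A simple grid generator projects evenly spaced points vertically onto the world using a custom trace channel.
# Tests filter by safe zone containment and inside-building state, check for flat, valid surfaces within a radius, and both filter and score on nearby buildings.
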EnvQuery /Game/Athena/Playlists/Mash/Spawners/Mash_EQS_RiftDrop_Static_Grid
options [{
  generator {
   name=EnvQueryGenerator_SimpleGrid_4
   class=EnvQueryGenerator_SimpleGrid
   properties {
    generateAround=/Script/FortniteAIServer.FortQueryContext_AthenaCurrentSafeZoneCenter
    gridSize=6000
    spaceBetween=1000
    projectionData {
     extentX=50
     navigationFilter=&RecastFilter_UseDefaultArea
     postProjectionVerticalOffset=0
     projectDown=3000
     projectUp=10000
     serializedChannel=ECC_GameTraceChannel11
     traceChannel=TraceTypeQuery10
    }
   }
  }
  tests [
   FortQueryTest_InsideAthenaSafeZone_4 {class=FortQueryTest_InsideAthenaSafeZone, properties {testPurpose=Filter}}
   FortQueryTest_ValidSurface_4 {
    class=FortQueryTest_ValidSurface
    properties {flatSurfaceToleranceZ=50, radius=360, traceOffsetDown=100, traceOffsetUp=100, surfaceMaterials []}
   }
   FortQueryTest_HasNearbyBuildings_4 {class=FortQueryTest_HasNearbyBuildings, properties {extentXY=1, extentZ=2, scoringFactor=20, testPurpose=FilterAndScore}}
   FortQueryTest_InsideBuilding_3 {class=FortQueryTest_InsideBuilding, properties {testPurpose=Filter}}
  ]
 }]
