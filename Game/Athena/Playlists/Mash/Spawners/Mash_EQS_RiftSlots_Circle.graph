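# Environment query in the Mash playlist's spawner folder that, per its name, finds rift slot locations on a circle.
# An OnCircle generator creates points at a parameter-driven radius, using navigation filters for trace and projection, with narrow projection extents and a small vertical lift afterward.
# A single filter-only 3D distance test enforces a parameter-driven minimum distance from the querier.
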EnvQuery /Game/Athena/Playlists/Mash/Spawners/Mash_EQS_RiftSlots_Circle
options [{
  generator {
   name=EnvQueryGenerator_OnCircle_10
   class=EnvQueryGenerator_OnCircle
   properties {
    bAutoSortTests=false
    numberOfPoints=8
    spaceBetween=250
    circleRadius {
     dataBinding=&self:EnvQueryGenerator_OnCircle_10.AIDataProvider_QueryParams_31
     dataField=FloatValue
     defaultValue=1000
    }
    projectionData {
     extentX=10
     navigationFilter=&RecastFilter_UseDefaultArea
     postProjectionVerticalOffset=5
     projectDown=115
     projectUp=115
    }
    traceData {navigationFilter=&RecastFilter_UseDefaultArea}
   }
  }
  tests [EnvQueryTest_Distance_0 {
    class=EnvQueryTest_Distance
    properties {
     distanceTo=&EnvQueryContext_Querier
     filterType=Minimum
     scoringEquation=Linear
     scoringFactor=0
     testComment=""
     testMode=Distance3D
     testPurpose=Filter
     floatValueMin {
      dataBinding=&self:EnvQueryTest_Distance_0.AIDataProvider_QueryParams_33
      dataField=FloatValue
      defaultValue=0
     }
    }
   }]
 }]
