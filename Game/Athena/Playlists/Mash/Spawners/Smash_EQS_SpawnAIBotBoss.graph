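# Environment query that finds spawn points for the AI bot boss in the Athena Mash playlist.
# A donut generator lays dense concentric rings in a band around the current safe zone center, projected over a tall vertical range.
# Two filter-only tests then screen the points by whether they are inside the safe zone and inside a building.
# No scoring test ranks the remaining candidates.
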EnvQuery /Game/Athena/Playlists/Mash/Spawners/Smash_EQS_SpawnAIBotBoss
options [{
  generator {
   name=EnvQueryGenerator_Donut_1
   class=EnvQueryGenerator_Donut
   properties {
    bUseSpiralPattern=false
    center=/Script/FortniteAIServer.FortQueryContext_AthenaCurrentSafeZoneCenter
    innerRadius=1024
    numberOfRings=16
    outerRadius=2048
    pointsPerRing=32
    projectionData {
     extentX=50
     navigationFilter=&RecastFilter_UseDefaultArea
     postProjectionVerticalOffset=0
     projectDown=3000
     projectUp=10000
     serializedChannel=ECC_GameTraceChannel11
     traceChannel=TraceTypeQuery10
    }
   }
  }
  tests [
   FortQueryTest_InsideAthenaSafeZone_0 {class=FortQueryTest_InsideAthenaSafeZone, properties {testPurpose=Filter}}
   FortQueryTest_InsideBuilding_1 {class=FortQueryTest_InsideBuilding, properties {testPurpose=Filter}}
  ]
 }]
